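# Sound cue for the chocolate biped cosmetic companion's landing foley, in the pets sound class with companion foley attenuation and the shared biped footstep concurrency.
# A quality-level node routes its first two levels to a random pick of six land waves and its last two to a random pick from only the first four.
# A root modulator adds slight volume variation, while cue multipliers lower the pitch and boost the volume.
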SoundCue /CosmeticCompanions/Assets/Audio_Shared/Foley/Biped/Chocolate/SC_MImosa_Biped_Chocolate_Land
root=Modulator_0
nodes [
 {id=Modulator_0, children [QualityLevel_0], properties {pitchMax=1, pitchMin=1, volumeMax=1, volumeMin=0.9}}
 {id=QualityLevel_0, children [Random_0,Random_0,Random_3,Random_3]}
 {id=Random_0, children [WavePlayer_0,WavePlayer_1,WavePlayer_2,WavePlayer_3,WavePlayer_4,WavePlayer_5], properties {weights [1,1,1,1,1,1]}}
 {id=Random_3, children [WavePlayer_6,WavePlayer_7,WavePlayer_8,WavePlayer_9], properties {weights [1,1,1,1]}}
 {
  id=WavePlayer_0
  properties {soundWaveAssetPtr=/CosmeticCompanions/Assets/Audio_Shared/Foley/Biped/Chocolate/Wavs/Mimosa_Biped_Chocolate_Land_01}
 }
 {
  id=WavePlayer_1
  properties {soundWaveAssetPtr=/CosmeticCompanions/Assets/Audio_Shared/Foley/Biped/Chocolate/Wavs/Mimosa_Biped_Chocolate_Land_02}
 }
 {
  id=WavePlayer_2
  properties {soundWaveAssetPtr=/CosmeticCompanions/Assets/Audio_Shared/Foley/Biped/Chocolate/Wavs/Mimosa_Biped_Chocolate_Land_03}
 }
 {
  id=WavePlayer_3
  properties {soundWaveAssetPtr=/CosmeticCompanions/Assets/Audio_Shared/Foley/Biped/Chocolate/Wavs/Mimosa_Biped_Chocolate_Land_04}
 }
 {
  id=WavePlayer_4
  properties {soundWaveAssetPtr=/CosmeticCompanions/Assets/Audio_Shared/Foley/Biped/Chocolate/Wavs/Mimosa_Biped_Chocolate_Land_05}
 }
 {
  id=WavePlayer_5
  properties {soundWaveAssetPtr=/CosmeticCompanions/Assets/Audio_Shared/Foley/Biped/Chocolate/Wavs/Mimosa_Biped_Chocolate_Land_06}
 }
 {
  id=WavePlayer_6
  properties {soundWaveAssetPtr=&Mimosa_Biped_Chocolate_Land_01}
 }
 {
  id=WavePlayer_7
  properties {soundWaveAssetPtr=&Mimosa_Biped_Chocolate_Land_02}
 }
 {
  id=WavePlayer_8
  properties {soundWaveAssetPtr=&Mimosa_Biped_Chocolate_Land_03}
 }
 {
  id=WavePlayer_9
  properties {soundWaveAssetPtr=&Mimosa_Biped_Chocolate_Land_04}
 }
]
properties {
 attenuationSettings=/Game/Sounds/Attenuation/AS_Companion_Foley
 maxDistance=1200
 pitchMultiplier=0.7
 priority=0
 soundClassObject=&Fort_Creatures_Pets
 volumeMultiplier=1.25
 concurrencySet [/CosmeticCompanions/Assets/Audio_Shared/Foley/Biped/CONC_Mimosa_Biped_Footstep]
}
